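# Translucent UI material for the quickbar that draws a looping animated frame preview with an optional chevron arrow.
# A time loop drives the frame inset between OuterSize and MinSize, with two SmoothSteps on a 9-sliced square distance-field texture cutting the border.
# The ShowArrow switch adds a pixel-scaled distance-field chevron via Max, which fades over the loop.
# OpacityRange and ArrowMoveRange each lerp R to R, so neither animates.
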Material /Game/UI/InGame/Quickbar/M_UI_FramePreview
graph {
 properties {
  bUsedWithStaticMesh=false
  blendMode=BLEND_Translucent
  materialDomain=MD_UI
  pixelDepthOffsetMode=PDOM_Legacy
  usedShadingModels=Default Lit
 }
 nodes [
  Constant() {r=1}
  n1 = MaterialFunctionCall() {materialFunction=&GetUserInterfaceUV}
  n2 = TextureSample(UVs=@n1.9-Slice UV) {texture=/Game/Athena/HUD/Materials/Score/T_UI_Square_DF}
  n3 = Divide(A=Fmod(A=Time() {period=1}, B=Constant() {r=1}), B=ScalarParameter() {defaultValue=1, parameterName=Speed})
  n4 = LinearInterpolate(A=ScalarParameter() {defaultValue=1, parameterName=OuterSize}, B=ScalarParameter() {defaultValue=0.9, parameterName=MinSize}, Alpha=@n3)
  n5 = OneMinus(None=@n4)
  n6 = ScalarParameter() {parameterName=EdgeSoftness}
  n7 = Add(A=@n5, B=@n6)
  n8 = ScalarParameter() {defaultValue=0.041, parameterName=BorderSize}
  n9 = Add(A=@n5, B=@n8)
  n10 = Add(A=@n9, B=@n6)
  n11 = VectorParameter() {parameterName=OpacityRange, defaultValue=<rgba 1 0 0 1>}
  n12 = Multiply(
   A=Subtract(
    A=MaterialFunctionCall(Alpha=@n2.R, Min=@n5, Max=@n7) {materialFunction=&SmoothStep}.Result
    B=MaterialFunctionCall(Alpha=@n2.R, Min=@n9, Max=@n10) {materialFunction=&SmoothStep}.Result
   )
   B=LinearInterpolate(A=@n11.R, B=@n11.R, Alpha=@n3)
  )
  n13 = MaterialFunctionCall() {materialFunction=&GetUserInterfaceUV}
  n14 = Divide(A=ScalarParameter() {defaultValue=32, group=Arrow, parameterName=ChevronSourceSize}, B=@n13.Pixel Size)
  n15 = VectorParameter() {group=Arrow, parameterName=ChevronOffset, defaultValue=<rgba 0.5 0 1 1>}
  n16 = Constant()
  n17 = VectorParameter() {group=Arrow, parameterName=ArrowMoveRange, defaultValue=<rgba 0.1 0 0 1>}
  n18 = LinearInterpolate(A=@n17.R, B=@n17.R, Alpha=@n3) {constA=0.1, constB=0}
  n19 = TextureSampleParameter2D(UVs=Add(
    A=MaterialFunctionCall!(Texture Scale=@n14, UVs=@n13.Normalized UV, Offset XY Texturesize Z=@n15.RGB) {materialFunction=/Engine/Functions/Engine_MaterialFunctions01/Texturing/ScaleUVsAroundPoint}.UVs
    B=AppendVector(A=@n16, B=@n18)
   )) {
   group=Arrow
   parameterName=ArrowTexture
   samplerSource=SSM_Clamp_WorldGroupSettings
   texture=/Game/Athena/HUD/LTM/GG/Materials/T_UI_Arrow_Chevron_DF
  }
  n20 = Multiply(A=OneMinus(None=@n3), B=@n19.R) {constA=1}
  n21 = ScalarParameter() {defaultValue=0.2, group=Arrow, parameterName=chevronEdge}
  n22 = Add(A=@n21, B=ScalarParameter() {group=Arrow, parameterName=EdgeSoftness})
  StaticSwitchParameter(
   True=Max(
    A=@n12
    B=MaterialFunctionCall(Alpha=@n20, Min=@n21, Max=@n22) {materialFunction=&SmoothStep}.Result
   )
   False=@n12
  ) {defaultValue=true, group=Arrow, parameterName=ShowArrow}
  Subtract(
   A=MaterialFunctionCall(UVs=@n1.9-Slice UV, Texture Scale=@n4) {materialFunction=&ScaleUVsByCenter}.0-1 mask
   B=MaterialFunctionCall(UVs=@n1.9-Slice UV, Texture Scale=Subtract(A=@n4, B=@n8)) {materialFunction=&ScaleUVsByCenter}.0-1 mask
  )
 ]
}
